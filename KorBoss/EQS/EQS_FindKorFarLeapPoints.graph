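# EQS query that picks a distant landing point for the Kor boss's leap relative to its target player.
# A custom Kor leap point generator supplies candidates, which are projected onto navigation and geometry.
# Filters require points inside the bot leash, minimum distances from querier and player, and passing trace and obstruction checks against them.
# Scoring favours points horizontally far from the player but near their height, plus a random factor.
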EnvQuery /KorBoss/EQS/EQS_FindKorFarLeapPoints
options [{
  generator {
   name=FortQueryGenerator_KorSecondaryLeapPoints_C_1
   class=FortQueryGenerator_KorLeapPoints_C
   properties {
    bAutoSortTests=true
    bCanRunAsync=false
    context=&/FortQueryContext_KorLeapPoints.FortQueryContext_KorLeapPoints_C
    envQueryResultNormalizationOption=Default
    generatedItemType=/Script/AIModule.EnvQueryItemType_Point
    generatorsActionDescription=""
    optionName=""
   }
  }
  tests [
   EnvQueryTest_Project_1 {
    class=EnvQueryTest_Project
    properties {
     scoringFactor=0
     testComment=""
     testPurpose=Score
     projectionData {navigationFilter=null, postProjectionVerticalOffset=0, projectDown=1024, projectUp=1024, traceMode=Navigation}
    }
   }
   EnvQueryTest_Trace_0 {
    class=EnvQueryTest_Trace
    properties {
     context=&/FortQueryContext_KorTargetPlayer.FortQueryContext_KorTargetPlayer_C
     scoringFactor=1
     testComment=""
     testPurpose=Filter
     itemHeightOffset {
      dataBinding=&self:EnvQueryTest_Trace_0.FortAIDataProvider_ScalableFloat_0
      dataField=FloatValue
      defaultValue=100
     }
    }
   }
   FortQueryTest_InsideAIBotLeash_0 {class=FortQueryTest_InsideAIBotLeash, properties {testPurpose=Filter}}
   FortQueryTest_IsObstructed_1 {
    class=FortQueryTest_IsObstructed
    properties {context=&FortQueryContext_KorTargetPlayer_C, testPurpose=Filter, traceFromContext=true}
   }
   FortQueryTest_IsObstructed_0 {class=FortQueryTest_IsObstructed, properties {context=&EnvQueryContext_Querier, testPurpose=Filter}}
   EnvQueryTest_Distance_2 {
    class=EnvQueryTest_Distance
    properties {
     distanceTo=&EnvQueryContext_Querier
     filterType=Minimum
     floatValueMax=5000
     floatValueMin=1000
     scoreClampMax=5000
     scoreClampMin=1000
     scoringEquation=InverseLinear
     scoringFactor=1
     testComment=""
     testMode=Distance2D
     testPurpose=Filter
    }
   }
   EnvQueryTest_Distance_0 {
    class=EnvQueryTest_Distance
    properties {
     distanceTo=&FortQueryContext_KorTargetPlayer_C
     filterType=Minimum
     floatValueMax=5000
     floatValueMin=500
     scoreClampMax=5000
     scoreClampMin=1000
     scoringEquation=Linear
     scoringFactor=1
     testComment=""
     testMode=Distance2D
     testPurpose=FilterAndScore
    }
   }
   EnvQueryTest_Distance_1 {
    class=EnvQueryTest_Distance
    properties {
     distanceTo=&FortQueryContext_KorTargetPlayer_C
     filterType=Minimum
     floatValueMax=5000
     floatValueMin=1000
     scoreClampMax=5000
     scoreClampMin=1000
     scoringEquation=InverseLinear
     scoringFactor=1
     testComment=""
     testMode=DistanceZ
     testPurpose=Score
    }
   }
   FortQueryTest_Random_1 {class=FortQueryTest_Random, properties {floatValueMax=0, scoringFactor=1}}
   EnvQueryTest_Project_2 {
    class=EnvQueryTest_Project
    properties {
     scoringFactor=0
     testComment=""
     testPurpose=Filter
     projectionData {
      navigationFilter=null
      postProjectionVerticalOffset=86.25
      projectDown=500
      projectUp=150
      serializedChannel=ECC_Visibility
      traceMode=GeometryByChannel
     }
    }
   }
  ]
 }]
