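# EQS query that finds short leap landing points for the Kor boss around its target player.
# A two-ring spiral donut generates candidates close around the target, without projection at generation.
# Filters keep points that pass a height-offset trace to the target, lie inside the AI bot leash and project onto navigation.
# Scoring favours points nearer the querier in 3D, with a random test adding variety.
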EnvQuery /KorBoss/EQS/EQS_FindKorShortLeapPoints
options [{
  generator {
   name=EnvQueryGenerator_Donut_2
   class=EnvQueryGenerator_Donut
   properties {
    bUseSpiralPattern=true
    center=&/FortQueryContext_KorTargetPlayer.FortQueryContext_KorTargetPlayer_C
    innerRadius=400
    numberOfRings=2
    outerRadius=700
    pointsPerRing=8
    projectionData {extentX=0, navigationFilter=null, postProjectionVerticalOffset=0, traceMode=None}
   }
  }
  tests [
   EnvQueryTest_Trace_4 {
    class=EnvQueryTest_Trace
    properties {
     context=&FortQueryContext_KorTargetPlayer_C
     scoringFactor=1
     testComment=""
     testPurpose=Filter
     itemHeightOffset {
      dataBinding=&self:EnvQueryTest_Trace_4.FortAIDataProvider_ScalableFloat_0
      dataField=FloatValue
      defaultValue=100
     }
    }
   }
   FortQueryTest_InsideAIBotLeash_0 {class=FortQueryTest_InsideAIBotLeash, properties {testPurpose=Filter}}
   EnvQueryTest_Distance_0 {
    class=EnvQueryTest_Distance
    properties {
     distanceTo=&EnvQueryContext_Querier
     filterType=Range
     floatValueMin=0
     scoringEquation=InverseLinear
     scoringFactor=1
     testComment=""
     testMode=Distance3D
     testPurpose=Score
    }
   }
   EnvQueryTest_Project_4 {
    class=EnvQueryTest_Project
    properties {
     scoringFactor=0
     testComment=""
     testPurpose=Filter
     projectionData {
      navigationFilter=null
      postProjectionVerticalOffset=86.25
      projectDown=500
      projectUp=150
      serializedChannel=ECC_Visibility
      traceMode=Navigation
     }
    }
   }
   FortQueryTest_Random_0 {class=FortQueryTest_Random, properties {floatValueMax=0, scoringFactor=1}}
  ]
 }]
