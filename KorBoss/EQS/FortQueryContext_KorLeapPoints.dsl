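; EQS query context that supplies Kor boss leap point locations on the island nearest its blackboard target.
; It finds the KB_IslandBox closest to the target and filters its tags by KBLeapPoint.IslandRelevance, using that parent tag if none is found.
; KB_LeapPoint actors with matching tags provide the returned locations.
; A failed querier cast only stores an EQS dummy as target, and the helper GetClosestIslandBoxTag is not called here.
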(fn GetClosestIslandBoxTag
  (params (KorTarget :type Actor%20Object%20Reference))
  (bind _closestislandbox (LocalVariables|GetClosestIslandBox))
  (LocalVariables|SetTargetLocation (Transformation|GetActorLocation KorTarget))
  (bind _outactors (Actor|GetAllActorsOfClass "/DemonArenaGameplay/KB/KB_IslandBox.KB_IslandBox_C"))
  (for _array_element _outactors
    (if (<= (.ReturnValue (Collision|GetClosestPointonCollision (Variables|/DemonArenaGameplay/KB/KB_IslandBox.KB_IslandBox_C|GetBox _array_element) (LocalVariables|GetTargetLocation) "None")) (LocalVariables|GetDistanceToTarget))
      (LocalVariables|SetClosestIslandBox _array_element)))
  (Utilities|IsValid _closestislandbox
    (:Is%20Valid
      (return (GameplayTags|Filter (Variables|/DemonArenaGameplay/KB/KB_IslandBox.KB_IslandBox_C|GetGameplayTags _closestislandbox) (GameplayTags|MakeGameplayTagContainerfromTag "(TagName=\"KBLeapPoint.IslandRelevance\")"))))
    (:Is%20Not%20Valid
      (return "(GameplayTags=((TagName=\"KBLeapPoint.IslandRelevance\")))"))))

(fn ProvideLocationsSet
  (params (QuerierObject :type Object%20Reference) (QuerierActor :type Actor%20Object%20Reference))
  (bind _targetlocation (LocalVariables|GetTargetLocation))
  (bind _closestislandbox (LocalVariables|GetClosestIslandBox))
  (bind _asbp_player_pawn_cherry_smoke_kor_boss (Utilities|Casting|CastToBP_PlayerPawn_CherrySmoke_KorBoss QuerierActor)
    (:then
      (bind _targetactor (Class|BPFLAICommonFunctions|GetTargetActorfromBlackboard _asbp_player_pawn_cherry_smoke_kor_boss))
      (Utilities|IsValid _targetactor
        (:Is%20Valid
          (LocalVariables|SetKorTarget _targetactor)
          (LocalVariables|SetTargetLocation (Transformation|GetActorLocation (LocalVariables|GetKorTarget)))
          (bind _outactors (Actor|GetAllActorsOfClass "/DemonArenaGameplay/KB/KB_IslandBox.KB_IslandBox_C"))
          (for _array_element _outactors
            (if (<= (.ReturnValue (Collision|GetClosestPointonCollision (Variables|/DemonArenaGameplay/KB/KB_IslandBox.KB_IslandBox_C|GetBox _array_element) _targetlocation "None")) (LocalVariables|GetDistanceToTarget))
              (LocalVariables|SetDistanceToTarget (.ReturnValue (Collision|GetClosestPointonCollision (Variables|/DemonArenaGameplay/KB/KB_IslandBox.KB_IslandBox_C|GetBox _array_element) _targetlocation "None")))
              (LocalVariables|SetClosestIslandBox _array_element)))
          (Utilities|IsValid _closestislandbox
            (:Is%20Valid
              (LocalVariables|SetTagToLookOutFor (GameplayTags|Filter (Variables|/DemonArenaGameplay/KB/KB_IslandBox.KB_IslandBox_C|GetGameplayTags _closestislandbox) (GameplayTags|MakeGameplayTagContainerfromTag "(TagName=\"KBLeapPoint.IslandRelevance\")"))))
            (:Is%20Not%20Valid
              (LocalVariables|SetTagToLookOutFor "(GameplayTags=((TagName=\"KBLeapPoint.IslandRelevance\")))")))
          (bind _outactors_2 (Actor|GetAllActorsOfClass "/DemonArenaGameplay/KB/KB_LeapPoint.KB_LeapPoint_C"))
          (for _array_element_2 _outactors_2
            (if (GameplayTags|HasAnyTags (Variables|/DemonArenaGameplay/KB/KB_LeapPoint.KB_LeapPoint_C|GetGameplayTags _array_element_2) (LocalVariables|GetTagToLookOutFor))
              (Utilities|Array|AddUnique (LocalVariables|GetPossibleLocations) (Transformation|GetActorLocation _array_element_2))))
          (return (LocalVariables|GetPossibleLocations)))
        (:Is%20Not%20Valid
          _)))
    (:CastFailed
      (bind _outactors_3 (Actor|GetAllActorsOfClass "/KorBoss/EQS/CherrySmoke_Kor_PlayerTargetEQSDummy.CherrySmoke_Kor_PlayerTargetEQSDummy_C"))
      (LocalVariables|SetKorTarget (Utilities|Array|Get(acopy) _outactors_3)))))
