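; EQS context for the Kor boss that supplies its target player as a single actor for environment queries.
; When the querier is the CherrySmoke Kor boss pawn, it returns the target actor read from that pawn's blackboard.
; If the cast fails, it instead finds a CherrySmoke_Kor_PlayerTargetEQSDummy actor as a stand-in target.
; Either result is returned only if valid.
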(fn ProvideSingleActor
  (params (QuerierObject :type Object%20Reference) (QuerierActor :type Actor%20Object%20Reference))
  (bind _asbp_player_pawn_cherry_smoke_kor_boss (Utilities|Casting|CastToBP_PlayerPawn_CherrySmoke_KorBoss QuerierActor)
    (:then
      (bind _targetactor (Class|BPFLAICommonFunctions|GetTargetActorfromBlackboard _asbp_player_pawn_cherry_smoke_kor_boss))
      (Utilities|IsValid _targetactor
        (:Is%20Valid
          (return _targetactor))))
    (:CastFailed
      (bind _returnvalue (Actor|GetActorOfClass "/KorBoss/EQS/CherrySmoke_Kor_PlayerTargetEQSDummy.CherrySmoke_Kor_PlayerTargetEQSDummy_C"))
      (Utilities|IsValid _returnvalue
        (:Is%20Valid
          (return _returnvalue))))))
