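; KorLeapPoints is an EQS query generator Blueprint that produces point items, apparently candidate leap locations for the Kor boss.
; Its DoItemGeneration override loops over the context locations it receives and adds each one as a generated vector.
; It passes the points through unchanged, with no offsets, projection or filtering, so the query context alone determines the candidate positions.
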(fn DoItemGeneration
  (params (ContextLocations :type Array%20of%20Vectors))
  (for _array_element ContextLocations
    (EQS|AddGeneratedVector _array_element))
  (return))
